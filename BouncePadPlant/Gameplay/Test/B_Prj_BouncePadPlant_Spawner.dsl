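; Test projectile for the bounce pad plant that spawns an actor where it touches something, then destroys itself.
; On touch, it breaks the hit result and, only with network authority, spawns the class held in ActorClassToSpawn at the hit location.
; The spawn transform takes only that location, and collision handling adjusts placement if possible but always spawns.
; The separate Hit event and the construction script are empty.
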(event EventOnTouched
  (params (OtherActor :type Actor%20Object%20Reference) (OtherComp :type Primitive%20Component%20Object%20Reference) (HitResult :type Hit%20Result%20Structure%20%28by%20ref%29) (bIsOverlap :type Boolean))
  (bind (_bblockinghit _binitialoverlap _time _distance _location) (Collision|BreakHitResult HitResult))
  (Networking|SwitchHasAuthority
    (:Authority
      (Game|SpawnActorfromClass (Variables|Default|GetActorClassToSpawn) (Math|Transform|MakeTransform _location) "AdjustIfPossibleButAlwaysSpawn" "SelectDefaultAtRuntime" 0 0))
    (:Remote
      _))
  (Actor|DestroyActor))

(event Collision|EventHit
  (params (MyComp :type Primitive%20Component%20Object%20Reference) (Other :type Actor%20Object%20Reference) (OtherComp :type Primitive%20Component%20Object%20Reference) (bSelfMoved :type Boolean) (HitLocation :type Vector :default "0, 0, 0") (HitNormal :type Vector :default "0, 0, 0") (NormalImpulse :type Vector :default "0, 0, 0") (Hit :type Hit%20Result%20Structure%20%28by%20ref%29)))

(fn ConstructionScript)
